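; BP_Gate_Quest_VehicleSpeed is a modular quest gate that opens while the quest owner moves forward in a vehicle at or above a speed threshold.
; On each DistanceTraveled event it resolves the primary quest owner to a player pawn and compares the vehicle interface's forward km/h speed to SpeedThreshold.
; It triggers the gate closed below that threshold, on owner or pawn lookup failure, or on an ExitVehicle event.
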(event ModularGates|EventVerbEventReceived
  (params (EventName :type Name :default "None") (EventDelta :type Integer))
  (switch name EventName
    (:DistanceTraveled
      (bind _outowneractor (ModularGates|GetPrimaryQuestOwner)
        (:Valid
          (bind _fortpawn (Class|FortQuestBlueprintLibrary|GetFortPawnFromObject _outowneractor)
            (:Valid
              (bind _asfort_player_pawn (Utilities|Casting|CastToFortPlayerPawn _fortpawn)
                (:then
                  (if (>= (Vehicle|GetForwardSpeedKmh (Vehicle|GetVehicleInterface _asfort_player_pawn)) (Variables|Default|GetSpeedThreshold))
                    (ModularGates|TriggerGateOpen)
                    (else
                      (ModularGates|TriggerGateClosed))))
                (:CastFailed
                  (ModularGates|TriggerGateClosed))))
            (:NotValid
              (ModularGates|TriggerGateClosed))))
        (:NotValid
          (ModularGates|TriggerGateClosed))))
    (:ExitVehicle
      (ModularGates|TriggerGateClosed))))
