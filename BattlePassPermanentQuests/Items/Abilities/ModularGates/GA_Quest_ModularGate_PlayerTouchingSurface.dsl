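; Modular gate ability for quests that opens or closes a gate based on the surface the participating player is touching.
; It handles participant gameplay events and acts only on surface-type-changed events.
; It checks the payload's instigator tags against a configured set of surface tags, triggering the gate open on any exact match and closed otherwise.
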(event ModularGates|EventParticipantReceivedGameplayEvent
  (params (Participant :type Object%20Reference) (EventTag :type Gameplay%20Tag%20Structure%20%28by%20ref%29) (Payload :type Gameplay%20Event%20Data%20Structure%20%28by%20ref%29))
  (bind (_eventtag _instigator _target _optionalobject _optionalobject2 _contexthandle _instigatortags) (Utilities|Struct|BreakGameplayEventData Payload))
  ;[0 0 0] When the player touches a particular surface, open the gate; else, close the gate
  (if (== EventTag "(TagName=\"Event.SurfaceType.Changed\")")
    (if (GameplayTags|HasAnyTags _instigatortags (Variables|Surface|GetSurfaceTagsToCheck) true)
      (ModularGates|TriggerGateOpen)
      (else
        (ModularGates|TriggerGateClosed)))))
